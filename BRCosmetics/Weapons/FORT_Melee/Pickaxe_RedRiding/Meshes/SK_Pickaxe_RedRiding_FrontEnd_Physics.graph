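# Physics asset for the front-end mesh of the RedRiding pickaxe, giving capsule bodies to its main body and dangling link and ribbon bones.
# Three constraint chains hang from the body: two links, five-segment ribbon C and six-segment ribbon D.
# The link joints lock twist and one swing axis, so they pivot like hinges.
# Ribbon D's last four joints have near-zero angle limit values, far tighter than the other joints.
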PhysicsAsset /BRCosmetics/Weapons/FORT_Melee/Pickaxe_RedRiding/Meshes/SK_Pickaxe_RedRiding_FrontEnd_Physics
bodies {
 body [{shapeType=Capsule, length=56.9578, radius=3.8402, center=<-0.0592 -0.1696 40.2661>, rotation=<rot 0.9512 -89.9168 1.2434>}]
 dyn_link1 [{shapeType=Capsule, length=0, radius=2.25, center=<3.1732 -0.6691 0.6732>, rotation=<rot 0 -90 0>}]
 dyn_link2 [{shapeType=Capsule, length=0, radius=7.2989, center=<9.9555 -0.2304 0.1636>, rotation=<rot 0 119.1624 -80.1583>}]
 dyn_ribbonC1 [{shapeType=Capsule, length=4.4693, radius=1.4598, center=<1.976 0.034 0.2995>, rotation=<rot 0 89.9216 -76.5374>}]
 dyn_ribbonC2 [{
   shapeType=Capsule
   length=2.3776
   radius=1.8647
   shapeName=dyn_ribbonC2_capsule
   center=<2.1178 0.092 0.0613>
   rotation=<rot -8.0861 97.7791 -87.818>
  }]
 dyn_ribbonC3 [{
   shapeType=Capsule
   length=3.291
   radius=2.034
   shapeName=dyn_ribbonC3_capsule
   center=<2.3953 0.3876 -0.0059>
   rotation=<rot -0.3483 86.9096 -88.9469>
  }]
 dyn_ribbonC4 [{
   shapeType=Capsule
   length=4.1675
   radius=1.6459
   shapeName=dyn_ribbonC4_capsule
   center=<2.7993 0.306 0.2808>
   rotation=<rot 0.0542 92.8394 -87.0439>
  }]
 dyn_ribbonC5 [{
   shapeType=Capsule
   length=0.5
   radius=1.505
   shapeName=dyn_ribbonC5_capsule
   center=<-0.0735 0.5959 0.4218>
   rotation=<rot 0 -90.0004 85.7564>
  }]
 dyn_ribbonD1 [{shapeType=Capsule, length=3.0814, radius=2.0643, center=<0.9099 0.4668 1.1432>, rotation=<rot 0 65.9483 -85.1764>}]
 dyn_ribbonD2 [{
   shapeType=Capsule
   length=0.8806
   radius=0.9575
   shapeName=dyn_ribbonD2_capsule
   center=<4.6956 -0.0285 -0.3998>
   rotation=<rot 0.0001 107.1814 -69.6271>
  }]
 dyn_ribbonD3 [{
   shapeType=Capsule
   length=1.505
   radius=1.755
   shapeName=dyn_ribbonD3_capsule
   center=<-0.1315 -3.5423 -0.6146>
   rotation=<rot 2.6598 174.9956 -85.1114>
  }]
 dyn_ribbonD4 [{
   shapeType=Capsule
   length=1.0342
   radius=2.25
   shapeName=dyn_ribbonD4_capsule
   center=<1.7212 -0.0298 -0.0662>
   rotation=<rot 1.3938 88.1244 -90.0004>
  }]
 dyn_ribbonD5 [{
   shapeType=Capsule
   length=0.505
   radius=2.005
   shapeName=dyn_ribbonD5_capsule
   center=<3.5531 0.1036 0.1216>
   rotation=<rot 52.3328 93.8727 -86.4988>
  }]
 dyn_ribbonD6 [{shapeType=Capsule, length=0.755, radius=1.505, shapeName=dyn_ribbonD6_capsule, center=<1.1829 0.2557 0>, rotation=<rot 0 -90.0001 -89.6368>}]
}
constraints [
 {
  bone1Name=dyn_link1
  bone2Name=body
  swing1LimitDegrees=45
  swing1Motion=Locked
  swing2LimitDegrees=45
  twistLimitDegrees=45
  twistMotion=Locked
 }
 {
  bone1Name=dyn_link2
  bone2Name=dyn_link1
  swing1LimitDegrees=45
  swing1Motion=Locked
  swing2LimitDegrees=45
  twistLimitDegrees=45
  twistMotion=Locked
 }
 {bone1Name=dyn_ribbonC1, bone2Name=body, swing1LimitDegrees=45, swing2LimitDegrees=45, twistLimitDegrees=45}
 {bone1Name=dyn_ribbonC2, bone2Name=dyn_ribbonC1, swing1LimitDegrees=45, swing2LimitDegrees=45, twistLimitDegrees=45}
 {bone1Name=dyn_ribbonC3, bone2Name=dyn_ribbonC2, swing1LimitDegrees=45, swing2LimitDegrees=45, twistLimitDegrees=45}
 {bone1Name=dyn_ribbonC4, bone2Name=dyn_ribbonC3, swing1LimitDegrees=45, swing2LimitDegrees=45, twistLimitDegrees=45}
 {bone1Name=dyn_ribbonC5, bone2Name=dyn_ribbonC4, swing1LimitDegrees=45, swing2LimitDegrees=45, twistLimitDegrees=45}
 {bone1Name=dyn_ribbonD1, bone2Name=body, swing1LimitDegrees=45, swing2LimitDegrees=45, twistLimitDegrees=45}
 {bone1Name=dyn_ribbonD2, bone2Name=dyn_ribbonD1, swing1LimitDegrees=45, swing2LimitDegrees=45, twistLimitDegrees=45}
 {bone1Name=dyn_ribbonD3, bone2Name=dyn_ribbonD2, swing1LimitDegrees=1, swing2LimitDegrees=1, twistLimitDegrees=1}
 {bone1Name=dyn_ribbonD4, bone2Name=dyn_ribbonD3, swing1LimitDegrees=1, swing2LimitDegrees=1, twistLimitDegrees=1}
 {bone1Name=dyn_ribbonD5, bone2Name=dyn_ribbonD4, swing1LimitDegrees=1, swing2LimitDegrees=1, twistLimitDegrees=1}
 {bone1Name=dyn_ribbonD6, bone2Name=dyn_ribbonD5, swing1LimitDegrees=1, swing2LimitDegrees=1, twistLimitDegrees=1}
]
