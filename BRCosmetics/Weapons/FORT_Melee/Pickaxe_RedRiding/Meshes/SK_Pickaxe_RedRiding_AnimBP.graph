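# Animation Blueprint for the RedRiding pickaxe that adds secondary physics motion to its dynamic bones.
# Two RigidBody branches share the cached input pose, and bIsInFrontEnd selects the front-end or regular physics asset.
# The blended pose runs through Anim Dynamics on ribbons A and B, then copies dyn_link3's translation to dyn_lampTrackerA.
# A separate ribbon C/D Anim Dynamics chain is unconnected to the output, and the event graph is empty.
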AnimBlueprint /BRCosmetics/Weapons/FORT_Melee/Pickaxe_RedRiding/Meshes/SK_Pickaxe_RedRiding_AnimBP
target_skeleton=&/SK_Pickaxe_RedRiding_Skeleton
graphs [
 AnimGraph {nodes [
   n1 = Anim Dynamics - Bone: dyn_ribbonC1(Alpha=1, ExternalForce="0,0,0") {
    angularDampingOverride=0.9998
    bGravityOverrideInSimSpace=true
    bOverrideAngularDamping=true
    bUseGravityOverride=true
    alphaBoolBlend {blendOption=Linear}
    boundBone {boneName=dyn_ribbonC1}
    componentAppliedLinearAccClamp=<5000 5000 5000>
    componentLinearAccScale=<1 1 1>
    componentLinearVelScale=<1 1 1>
    gravityOverride {z=-981}
    physicsBodyDefinitions [{
      sphereCollisionRadius=0
      boundBone {boneName=dyn_ribbonC1}
      boxExtents=<10 16 1>
      constraintSetup {angularLimitsMax=<17.6168 56.3737 52.8504>, angularLimitsMin=<-95.1306 -7.0468 -52.8503>, angularTarget=<0 0 0>}
      localJointOffset=<0 0 -12>
     }]
   }
   n2 = Anim Dynamics - Bone: dyn_ribbonC2(ComponentPose=@n1, Alpha=1, ExternalForce="0,0,0") {
    angularDampingOverride=0.9998
    bGravityOverrideInSimSpace=true
    bOverrideAngularDamping=true
    bUseGravityOverride=true
    alphaBoolBlend {blendOption=Linear}
    boundBone {boneName=dyn_ribbonC2}
    componentAppliedLinearAccClamp=<5000 5000 5000>
    componentLinearAccScale=<1 1 1>
    componentLinearVelScale=<1 1 1>
    gravityOverride {z=-981}
    physicsBodyDefinitions [{
      sphereCollisionRadius=0
      boundBone {boneName=dyn_ribbonC2}
      boxExtents=<14 10 2>
      constraintSetup {angularLimitsMax=<59.8972 56.3737 52.8504>, angularLimitsMin=<-42.2802 -59.897 -52.8503>, angularTarget=<0 0 0>}
      localJointOffset=<0 0 -12>
     }]
   }
   n3 = Anim Dynamics - Bone: dyn_ribbonD1(ComponentPose=@n2, Alpha=1, ExternalForce="0,0,0") {
    angularDampingOverride=0.9998
    bGravityOverrideInSimSpace=true
    bOverrideAngularDamping=true
    bUseGravityOverride=true
    alphaBoolBlend {blendOption=Linear}
    boundBone {boneName=dyn_ribbonD1}
    componentAppliedLinearAccClamp=<5000 5000 5000>
    componentLinearAccScale=<1 1 1>
    componentLinearVelScale=<1 1 1>
    gravityOverride {z=-981}
    physicsBodyDefinitions [{
      sphereCollisionRadius=0
      boundBone {boneName=dyn_ribbonD1}
      boxExtents=<10 16 1>
      constraintSetup {twistAxis=AxisZ, angularLimitsMax=<17.6168 56.3737 52.8504>, angularLimitsMin=<-95.1306 0 -52.8503>, angularTarget=<0 0 0>}
      localJointOffset=<0 0 -12>
     }]
   }
   Anim Dynamics - Bone: dyn_ribbonD2(ComponentPose=@n3, Alpha=1, ExternalForce="0,0,0") {
    angularDampingOverride=0.9998
    bGravityOverrideInSimSpace=true
    bOverrideAngularDamping=true
    bUseGravityOverride=true
    alphaBoolBlend {blendOption=Linear}
    boundBone {boneName=dyn_ribbonD2}
    componentAppliedLinearAccClamp=<5000 5000 5000>
    componentLinearAccScale=<1 1 1>
    componentLinearVelScale=<1 1 1>
    gravityOverride {z=-981}
    physicsBodyDefinitions [{
      sphereCollisionRadius=0
      boundBone {boneName=dyn_ribbonD2}
      boxExtents=<14 10 2>
      constraintSetup {twistAxis=AxisZ, angularLimitsMax=<59.8972 56.3737 52.8504>, angularLimitsMin=<-42.2802 -59.897 -52.8503>, angularTarget=<0 0 0>}
      localJointOffset=<0 0 -12>
     }]
   }
   n4 = RigidBody(ComponentPose=Local To Component(LocalPose=Use cached pose 'INPUT'())) {
    bOverrideWorldGravity=true
    overridePhysicsAsset=&/SK_Pickaxe_RedRiding_FrontEnd_Physics
    alphaBoolBlend {blendOption=Linear}
    componentAppliedLinearAccClamp=<500 500 500>
    componentLinearAccScale=<1 1 1>
    componentLinearVelScale=<1 1 1>
    overrideWorldGravity {z=-981}
   }
   n5 = Component To Local(ComponentPose=@n4)
   n6 = RigidBody(ComponentPose=Local To Component(LocalPose=Use cached pose 'INPUT'())) {
    bOverrideWorldGravity=true
    overridePhysicsAsset=&/SK_Pickaxe_RedRiding_Physics
    alphaBoolBlend {blendOption=Linear}
    componentAppliedLinearAccClamp=<500 500 500>
    componentLinearAccScale=<1 1 1>
    componentLinearVelScale=<1 1 1>
    overrideWorldGravity {z=-981}
   }
   n7 = Component To Local(ComponentPose=@n6)
   n8 = Get bIsInFrontEnd()
   n9 = Anim Dynamics - Bone: dyn_ribbonA(
    ComponentPose=Local To Component(LocalPose=Blend Poses by bool(BlendPose_0=@n5, BlendPose_1=@n7, bActiveValue=@n8) {blendType=Linear})
    Alpha=1
    ExternalForce="0,0,0"
   ) {
    angularDampingOverride=0.999
    bGravityOverrideInSimSpace=true
    bOverrideAngularDamping=true
    bUseGravityOverride=true
    alphaBoolBlend {blendOption=Linear}
    boundBone {boneName=dyn_ribbonA}
    componentAppliedLinearAccClamp=<5000 5000 5000>
    componentLinearAccScale=<1 1 1>
    componentLinearVelScale=<1 1 1>
    gravityOverride {z=-981}
    physicsBodyDefinitions [{
      sphereCollisionRadius=0
      boundBone {boneName=dyn_ribbonA}
      boxExtents=<1 9 11>
      constraintSetup {twistAxis=AxisZ, angularLimitsMax=<17.6168 56.3737 52.8504>, angularLimitsMin=<-95.1306 -59.897 -52.8503>, angularTarget=<0 0 0>}
      localJointOffset=<0 0 -1>
     }]
   }
   n10 = Anim Dynamics - Bone: dyn_ribbonB(ComponentPose=@n9, Alpha=1, ExternalForce="0,0,0") {
    angularDampingOverride=0.999
    bGravityOverrideInSimSpace=true
    bOverrideAngularDamping=true
    bUseGravityOverride=true
    alphaBoolBlend {blendOption=Linear}
    boundBone {boneName=dyn_ribbonB}
    componentAppliedLinearAccClamp=<5000 5000 5000>
    componentLinearAccScale=<1 1 1>
    componentLinearVelScale=<1 1 1>
    gravityOverride {z=-981}
    physicsBodyDefinitions [{
      sphereCollisionRadius=0
      boundBone {boneName=dyn_ribbonB}
      boxExtents=<1 7 9>
      constraintSetup {twistAxis=AxisZ, angularLimitsMax=<17.6168 56.3737 52.8504>, angularLimitsMin=<-95.1306 -59.897 -52.8503>, angularTarget=<0 0 0>}
      localJointOffset=<0 0 -1>
     }]
   }
   n11 = Copy Bone - Source Bone: dyn_link3 - Target Bone: dyn_lampTrackerA!(ComponentPose:Component Space Pose Link Structure=@n10, Alpha:float=1) -> (Pose:Component Space Pose Link Structure) {
    is=title AnimGraphNode_CopyBone
    bCopyTranslation=true
    alphaBoolBlend {blendOption=Linear}
    sourceBone {boneName=dyn_link3}
    targetBone {boneName=dyn_lampTrackerA}
   }
   Output Pose(Result=Component To Local(ComponentPose=@n11)) {layerGroup=None}
   Save cached pose 'INPUT'(Pose=Input Pose()) {cacheName=INPUT}
  ]}
 EventGraph {nodes [Event BlueprintUpdateAnimation()]}
]
variables []
